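; Player gameplay ability that fakes a ragdoll with prototype montages, from a looping fall pose to a get-up.
; On activation it plays the loop, waits briefly for the MovementMode.Falling tag to be added, then waits for its removal.
; After a further delay it stops the loop, starts the get-up montage and ends immediately without awaiting completion.
; Ending removes the granting effect; debug prints mark start and end.
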(event Ability|EventActivateAbility
  (Development|PrintString "RAGDOLL START")
  (Ability|Tasks|PlayMontageAndWaitNotify "None" "/Game/Abilities/Ragdoll/ProtoRagdoll_loop_CMM_M.ProtoRagdoll_loop_CMM_M")
  ;[1 1 1] Delay so that MovementMode.FallingTag has time to be added
  (Utilities|FlowControl|Delay 0.2)
  (Ability|Tasks|WaitGameplayTagRemove "(TagName=\"MovementMode.Falling\")"
    (:Removed
      (Utilities|FlowControl|Delay 1.0)
      (Ability|Animation|MontageStop -1.0)
      (Ability|Tasks|PlayMontageAndWaitNotify "None" "/Game/Abilities/Ragdoll/PROTO_Ragdoll_GetUp_CMM_M.PROTO_Ragdoll_GetUp_CMM_M"
        (:OnCompleted
          _)
        (:OnInterrupted
          _)
        (:OnCancelled
          _))
      (Ability|EndAbility))))

(event Ability|EventOnEndAbility
  (params (bWasCancelled :type Boolean))
  (Development|PrintString "RAGDOLL END")
  (Ability|RemoveGrantedbyEffect))

(event Ability|EventFailedToActivatePassiveAbility
  (params (ActorInfo :type Gameplay%20Ability%20Actor%20Info%20Structure))
  (|CallParentFunction))
